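# Niagara system for the CrawlyTech backpack cosmetic, adding glow sprites and yellow-green lighting.
# Glow bursts three spawn groups at fixed offsets, using two sprite renderers and a light renderer, with intensity ramped in by system age.
# ConeGlow continuously spawns short-lived sprites at two oppositely aligned points, fading them through color and alpha curves.
# System-level color and brightness values feed both emitters, including a time-of-day multiplier on cone alpha.
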NiagaraSystem /BRCosmetics/Accessories/FORT_Backpacks/Backpack_M_MED_CrawlyTech/FX/NS_Backpack_CrawlyTech
dependencies {
 usedDataInterfaces [&NiagaraDataInterfaceCurve, &NiagaraDataInterfaceColorCurve]
 usedDynamicInputs [&Waveform, &FNTimeOfDayScalar]
 usedModules [
  &self:SystemSpawnScript.SystemScriptSource.SystemScriptGraph.NiagaraNodeAssignment_1.SetVariables_#1
  &SystemState
  &self:SystemSpawnScript.SystemScriptSource.SystemScriptGraph.NiagaraNodeAssignment_0.SetVariables_#2
  &self:SystemSpawnScript.SystemScriptSource.SystemScriptGraph.NiagaraNodeAssignment_2.SetVariables_#3
 ]
}
emitters {
 ConeGlow {
  emitterUpdateScript [
   &EmitterState()
   &SpawnRate(SpawnRate=8, Use Spawn Probability=false)
  ]
  particleSpawnScript [
   &FN_SetSpawnGroup(Number of Spawn Groups=2)
   &InitializeParticle(
    Lifetime=&RandomRangeFloat(Minimum=1, Maximum=1.5)
    Position Offset=&MergeHelper_VectorSwitch(0=<-10 -16 -15>, 1=<8 -5 14>, 2=?, 3=?, Index=&SpawnGroup(), Num=::NewEnumerator0)
    Sprite Size Max=<30 25>
    Sprite Size Min=<26 20>
    Sprite Size Mode=::NewEnumerator2
    UsePositionOffset=true
    Write Material Random=true
   )
   &SubUVAnimation(End Frame=4, SubUV Animation Mode:ENiagara_SubUVLookupMode=::NewEnumerator2)
  ]
  particleUpdateScript [
   &ParticleState()
   &SpriteFacingAndAlignment(
    Sprite Alignment=&MergeHelper_VectorSwitch(0=<0.7 0.5 1>, 1=<-0.6 -0.3 -1>, 2=?, 3=?, Index=&SpawnGroup(), Num=::NewEnumerator0)
    WriteAlignment=true
    WriteFacing=false
   )
   &ScaleColor(
    Scale Alpha=&FloatFromCurve(
     FloatCurve=DataInterface() {propertyValues {Curve {keys [{interpMode=RCIM_Cubic, tangentMode=RCTM_User, time=0, value=0}, {interpMode=RCIM_Cubic, tangentMode=RCTM_User, time=0.3, value=1}, <at 1 0>]}}}
     Scale Curve=&Multiply_Float(A=5, B=$System.TODBrightnessMultiplier)
    )
    Scale RGB=&MakeVectorFromLinearColorRGB(Color=&ColorFromCurve(ColorCurve=DataInterface() {propertyValues {
        AlphaCurve {keys [<at 0 1 RCIM_Cubic>,<at 1 1 RCIM_Cubic>]}
        BlueCurve {keys [<at 0.4 0.02 RCIM_Cubic>,<at 0.7 0.02>]}
        GreenCurve {keys [<at 0.4 1 RCIM_Cubic>,<at 0.7 1>]}
        RedCurve {keys [<at 0.4 0.8367 RCIM_Cubic>,<at 0.7 0.6733>]}
       }}))
    ScaleA=true
    ScaleRGB=true
   )
   &ScaleSpriteSize(Scale Sprite Size Mode=::NewEnumerator0, Uniform Scale Factor=0.65)
  ]
  rendererClasses [&NiagaraSpriteRendererProperties]
  renderers [{rendererClass=&NiagaraSpriteRendererProperties}]
 }
 Glow {
  emitterUpdateScript [
   &EmitterState(
    Inactive Response=::NewEnumerator1
    Life Cycle Mode=::NewEnumerator1
    Loop Behavior=::NewEnumerator1
    Loop Duration Mode=::NewEnumerator1
    UseLoopDelay=false
   )
   &SpawnBurst_Instantaneous(Spawn Count=3, Use Spawn Probability=false, UseLoopCountLimit=false)
   /SetVariables(Emitter.GlowColor:LinearColor=&MultiplyLinearColor(
     Linear Color=$System.Color
     Scale Factor=&MakeFloat4FromVectorAndFloat(
      "Vector (RGB)"=&VectorFromFloat(Value=$System.BrightnessMultiplier)
      "Float (Alpha)"=&FloatFromCurve(
       FloatCurve=DataInterface() {propertyValues {Curve {keys [
           {interpMode=RCIM_Cubic, leaveTangentWeight=0.3333, tangentMode=RCTM_User, time=0, value=0}
           {arriveTangent=5.0781, arriveTangentWeight=0.9792, interpMode=RCIM_Cubic, leaveTangent=5.0781, tangentMode=RCTM_User, time=0.5, value=1}
          ]}}}
       CurveIndex=$System.Age
       Scale Curve=0.3
      )
     )
    )) {isSetParameters=true}
  ]
  particleSpawnScript [
   &FN_SetSpawnGroup(Number of Spawn Groups=3)
   &InitializeParticle(
    Color Mode=::NewEnumerator0
    Lifetime=0
    Position Offset=&MergeHelper_VectorSwitch(0=<-6 -13 -13>, 1=<5 -6.5 9>, 2=<0 -9 -2>, 3=?, Index=&SpawnGroup(), Num=::NewEnumerator1)
    Sprite Rotation Angle=&MergeHelper_FloatSwitch(0=0, 1=0, 2=-15, 3=?, Index=&SpawnGroup(), Num=::NewEnumerator1)
    Sprite Rotation Mode=::NewEnumerator0
    Sprite Size=&MergeHelper_Vector2Switch(0=<25 25>, 1=<25 25>, 2=<40 70>, 3=?, 4=?, Index=&SpawnGroup(), Num=::NewEnumerator1)
    Sprite Size Mode=::NewEnumerator4
    UsePositionOffset=true
    Write Material Random=true
   )
   &SpriteFacingAndAlignment(
    Sprite Alignment=&MergeHelper_VectorSwitch(0=<0 1 0>, 1=<0 1 0>, 2=<0.1 0.5 1>, 3=?, Index=&SpawnGroup(), Num=::NewEnumerator1)
    Sprite Facing=&MergeHelper_VectorSwitch(0=<0 -1 0.1>, 1=<0 -1 0.1>, 2=<0 0 0>, 3=?, Index=&SpawnGroup(), Num=::NewEnumerator1)
    WriteAlignment=true
    WriteFacing=false
   )
   &CameraOffset(Camera Offset Amount=&MergeHelper_FloatSwitch(0=8, 1=8, 2=10, 3=?, Index=&SpawnGroup(), Num=::NewEnumerator1))
   &DynamicMaterialParameters(
    Index 0 Param 1=1
    Module.Index0Param0Write=true
    Module.Index0Param1Write=false
    Module.Index0Param2Write=false
    Module.Index0Param3Write=false
   )
  ]
  particleUpdateScript [
   /SetVariables(
    Particles.BigGlowColor:LinearColor=&MakeLinearColorFromVectorAndFloat(
     "Vector (RGB)"=&MakeVectorFromLinearColorRGB(Color=$Emitter.GlowColor)
     "Float (Alpha)"=&FloatFromCurve(
      FloatCurve=DataInterface() {propertyValues {Curve {keys [
          {interpMode=RCIM_Cubic, leaveTangentWeight=0.3333, tangentMode=RCTM_User, time=0, value=0}
          {arriveTangent=5.0781, arriveTangentWeight=0.9792, interpMode=RCIM_Cubic, leaveTangent=5.0781, tangentMode=RCTM_User, time=0.5, value=1}
         ]}}}
      CurveIndex=$System.Age
      Scale Curve=0.15
     )
    )
    Particles.VisibilityTag=&MergeHelper_IntSwitch(0=0, 1=0, 2=1, 3=?, 4=?, 5=?, Index=&SpawnGroup(), Num=::NewEnumerator1)
   ) {isSetParameters=true}
   &Light_Attributes(
    Color=1
    Color Mode=0
    Enabled=1
    Light Color:LinearColor=&Multiply_LinearColorByFloat(
     Linear Color=<rgba 0.7625 1 0.05 1>
     Float=&FloatFromCurve(
      FloatCurve=DataInterface() {propertyValues {Curve {keys [
          {interpMode=RCIM_Cubic, leaveTangentWeight=0.3333, tangentMode=RCTM_User, time=0, value=0}
          {arriveTangent=5.0781, arriveTangentWeight=0.9792, interpMode=RCIM_Cubic, leaveTangent=5.0781, tangentMode=RCTM_User, time=0.5, value=1}
         ]}}}
      CurveIndex=$System.Age
      Scale Curve=500
     )
    )
    Light Position Offset:Vector3f=<0 -2 0>
    Light Radius=80
    Position=1
    Radius=1
    UsePositionOffset=true
   ) {module=Light_Attributes001}
  ]
  rendererClasses [&NiagaraSpriteRendererProperties, &NiagaraLightRendererProperties]
  renderers [
   {rendererClass=&NiagaraSpriteRendererProperties}
   {rendererClass=&NiagaraSpriteRendererProperties, rendererIndex=1}
   {rendererClass=&NiagaraLightRendererProperties, rendererIndex=2}
  ]
 }
}
summary {
 systemName=NS_Backpack_CrawlyTech
 emitters [
  {emitterName=Glow, rendererClasses [&NiagaraSpriteRendererProperties, &NiagaraLightRendererProperties]}
  {emitterName=ConeGlow, rendererClasses [&NiagaraSpriteRendererProperties]}
 ]
 userVariables [
  User.bIsFrontEnd {type=NiagaraBool}
  User.bIsFrontEndPreview {type=NiagaraBool}
 ]
}
