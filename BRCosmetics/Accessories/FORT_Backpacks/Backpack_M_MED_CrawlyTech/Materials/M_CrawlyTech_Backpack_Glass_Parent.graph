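# Translucent parent glass material for the CrawlyTech backpack cosmetic, with a liquid tint.
# MF_Glass and a Fresnel-driven thin-film layer share a TechArt mask, followed by optional camera dither.
# A static switch selects either VertexHide, which offsets vertices along the vector to the object pivot, or the character opacity function.
# DepthFade blends emissive toward LiquidColor as depth behind the surface increases, before MF_TechArtOverrides finishes the output.
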Material /BRCosmetics/Accessories/FORT_Backpacks/Backpack_M_MED_CrawlyTech/Materials/M_CrawlyTech_Backpack_Glass_Parent
graph {
 properties {
  bEnableResponsiveAA=true
  bScreenSpaceReflections=true
  bUseMaterialAttributes=true
  bUsedWithMeshDeformer=true
  bUsedWithMorphTargets=true
  bUsedWithSkeletalMesh=true
  blendMode=BLEND_Translucent
  ditherOpacityMask=true
  pixelDepthOffsetMode=PDOM_Legacy
  translucencyDirectionalLightingIntensity=5
  translucencyLightingMode=TLM_SurfacePerPixelLighting
  usedShadingModels=Default Lit
 }
 nodes [
  n1 = TextureSampleParameter2D() {
   group=00_Glass
   parameterName=TechArtMask
   samplerType=SAMPLERTYPE_Masks
   sortPriority=-96
   texture=&T_BlackMask
  }
  n2 = MaterialFunctionCall!(
   Material Attributes=MaterialFunctionCall!(TechArtMask=@n1.RGB) {materialFunction=/Game/Global/Materials/Character_Parent_2023/Glass_Parent_2023/MF_Glass}.Result
   Fresnel=Fresnel() {baseReflectFraction=0, exponent=1}
   Alpha=@n1.RGB
  ) {materialFunction=/Game/Global/MaterialFunctions/TechArt/MF_ThinFilm_TechArt}
  n3 = StaticSwitchParameter(
   True=MaterialFunctionCall!(In=@n2.Result) {materialFunction=/Game/Packages/Fortress_SharedMaterials/Base_Material_Functions/BaseCustomizationFunctions/MF_FN_Vehicle_CameraDither}.Result
   False=@n2.Result
  ) {group=00_Glass, parameterName=UseCameraDither, sortPriority=250}
  n4 = Multiply(A=Subtract(A=ObjectPositionWS(), B=WorldPosition().XYZ), B=ScalarParameter() {group=00_Glass, parameterName=VertexHide, sortPriority=301})
  n5 = GetMaterialAttributes(None=StaticSwitchParameter(
    True=SetMaterialAttributes!(MaterialAttributes=@n3, World Position Offset=@n4) {attributeSetTypes [#1]}
    False=MaterialFunctionCall!(Material=@n3) {materialFunction=/Game/Global/Materials/Character_Parent_2023/Functions/BaseFN_CharacterOpacity_2023}.Result
   ) {group=00_Glass, parameterName=UseVertexHide, sortPriority=300}) {attributeGetTypes [#2]}
  n6 = VectorParameter() {parameterName=LiquidColor, defaultValue=<rgba 0.4246 0.4583 0.135 1>}
  n7 = DepthFade(FadeDistance=ScalarParameter() {defaultValue=40, parameterName=FadeDistance})
  MaterialFunctionCall(In=SetMaterialAttributes!(MaterialAttributes=@n5.MaterialAttributes, Emissive Color=LinearInterpolate(A=@n5.Emissive Color, B=@n6.RGB, Alpha=@n7), Refraction) {attributeSetTypes [#2, #3]}) {materialFunction=&MF_TechArtOverrides}
 ]
}
