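# Parent material for the CrawlyTech backpack's liquid, an opaque, mostly emissive surface with animated bubbles and a reactive glow.
# Two panning bubble layers, distorted by a scrolling flowmap and limited by a smoothstep gradient, drive bubble color and blended normals.
# Emissive combines inner/outer colors, a Fresnel-shaped reactive tint and bubbles, scaled by time of day.
# A curve atlas turns time since a Blueprint-fed timestamp into dilation progress, with a debug override switch.
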Material /BRCosmetics/Accessories/FORT_Backpacks/Backpack_M_MED_CrawlyTech/Materials/M_CrawlyTech_Backpack_Liquid_Parent
graph {
 properties {
  bUseMaterialAttributes=true
  bUsedWithMeshDeformer=true
  bUsedWithMorphTargets=true
  bUsedWithSkeletalMesh=true
  pixelDepthOffsetMode=PDOM_Legacy
  usedShadingModels=Default Lit
 }
 nodes [
  n1 = Constant() {r=0.5}
  n2 = SetMaterialAttributes!(MaterialAttributes, Emissive Color=NamedRerouteUsage(), Normal=NamedRerouteUsage(), Base Color=Constant(), Specular=@n1, Roughness=@n1) {attributeSetTypes [
    #1
    #2
    #3
    #4
    #5
   ]}
  MaterialFunctionCall(In=@n2) {materialFunction=&MF_TechArtOverrides}
  n3 = SmoothStep(Min=Constant() {r=5}, Max=Constant() {r=7.5}, Value=ComponentMask(None=NamedRerouteUsage()) {g=true})
  n4 = NamedRerouteUsage()
  n5 = ScalarParameter() {defaultValue=0.75, group=CrawlyTech Backpack, parameterName=Potion-BubblesTile1}
  n6 = Multiply(A=Time(), B=ScalarParameter() {defaultValue=0.2, group=CrawlyTech Backpack, parameterName=Potion-BubblesSpeed})
  n7 = ScalarParameter() {defaultValue=2, group=CrawlyTech Backpack, parameterName=Potion-FlowmapScale}
  n8 = Time()
  n9 = ScalarParameter() {defaultValue=0.05, group=CrawlyTech Backpack, parameterName=Potion-FlowmapSpeed}
  n10 = TextureSample(UVs=Panner(Coordinate=Multiply(A=@n4, B=@n7), Time=Multiply(A=@n8, B=@n9)) {speedX=1}) {samplerType=SAMPLERTYPE_LinearColor, texture=/Game/Global/Textures/Default/Flowmaps/T_TorusFlowmap}
  n11 = ComponentMask(None=Multiply(
    A=Subtract(A=@n10.RGB) {constB=0.5}
    B=VectorParameter() {group=CrawlyTech Backpack, parameterName=Potion-FlowmapAmt, defaultValue=<rgba 1 1 0 1>}.RGB
   )) {g=true, r=true}
  n12 = Add(
   A=Panner(Coordinate=Multiply(A=@n4, B=@n5), Time=@n6) {speedX=-0.1, speedY=1}
   B=Multiply(A=ScalarParameter() {defaultValue=0.05, group=Potion, parameterName=Potion-FlowmapStrength1}, B=@n11)
  )
  n13 = TextureSample(UVs=@n12) {samplerSource=SSM_Wrap_WorldGroupSettings, samplerType=SAMPLERTYPE_Masks, texture=/Game/Global/Textures/Default/PointSplatter/Bubbles_M}
  n14 = ScalarParameter() {defaultValue=1, group=CrawlyTech Backpack, parameterName=Potion-BubblesTile2}
  n15 = Add(
   A=Panner(Coordinate=Multiply(A=@n4, B=@n14), Time=@n6) {speedX=0.1, speedY=1}
   B=Multiply(A=@n11, B=ScalarParameter() {defaultValue=0.025, group=CrawlyTech Backpack, parameterName=Potion-FlowmapStrength2})
  )
  n16 = TextureSample(UVs=@n15) {samplerSource=SSM_Wrap_WorldGroupSettings, samplerType=SAMPLERTYPE_Masks, texture=&Bubbles_M}
  n17 = Max(A=@n13.RGB, B=@n16.RGB)
  n18 = OneMinus(None=@n17)
  n19 = ComponentMask(None=Multiply(A=Power(Base=@n18) {constExponent=4}, B=@n17)) {g=true}
  n20 = ScalarParameter() {defaultValue=7, group=CrawlyTech Backpack, parameterName=Potion-BubbleMaskAmount}
  n21 = Multiply(A=Multiply(A=@n3, B=Multiply(A=@n19, B=@n20)), B=OneMinus(None=NamedRerouteUsage()))
  NamedRerouteDeclaration(None=Multiply(
    A=VectorParameter() {group=CrawlyTech Backpack, parameterName=BubbleColor, defaultValue=<rgba 1.4054 1.5 0.4217 1>}.RGB
    B=NamedRerouteDeclaration(None=@n21) {name=BubbleMask, nodeColor=<rgba 0 1 0.8594 1>}
   )) {name=BubblesColor, nodeColor=<rgba 1 0.7266 0 1>}
  n22 = VectorParameter() {group=CrawlyTech Backpack, parameterName=EmissiveInner, sortPriority=60, defaultValue=<rgba 0 0 0 1>}
  n23 = VectorParameter() {group=CrawlyTech Backpack, parameterName=EmissiveOuter, sortPriority=61, defaultValue=<rgba 0.9948 0.9274 0.7927 1>}
  n24 = NamedRerouteUsage()
  n25 = OneMinus(None=Fresnel(ExponentIn=Constant() {r=0.15}) {exponent=1})
  n26 = Constant() {r=0.1}
  n27 = VectorParameter() {group=CrawlyTech Backpack, parameterName=ReactiveEmissiveTint, sortPriority=61, defaultValue=<rgba 8.4455 9 3.6 1>}
  n28 = Multiply(A=Saturate(None=Add(A=@n25, B=@n26)), B=@n27.RGB)
  n29 = NamedRerouteUsage()
  n30 = Add(
   A=Multiply(A=LinearInterpolate(A=@n22.RGB, B=@n23.RGB, Alpha=@n24), B=LinearInterpolate(B=@n28, Alpha=@n29) {constA=1, constB=1.5})
   B=NamedRerouteUsage()
  )
  n31 = ScalarParameter() {defaultValue=1, group=CrawlyTech Backpack, parameterName=EmissiveStrength, sortPriority=62}
  n32 = Constant() {r=1}
  n33 = Constant() {r=0.75}
  NamedRerouteDeclaration(None=Multiply(
    A=Multiply(A=@n30, B=@n31)
    B=MaterialFunctionCall(Color_Morning=@n32, Color_Day=@n32, Color_Evening=@n33, Color_Night=@n33) {materialFunction=&Color_Time_of_Day}.Result
   )) {name=Emissive_OUT, nodeColor=<rgba 0.6406 1 0 1>}
  n34 = TextureCoordinate()
  NamedRerouteDeclaration(None=Add(A=Divide(A=@n34) {constB=0.1094}) {constB=0.1193}) {name=Bubbles_UVs, nodeColor=<rgba 0 0.9297 1 1>}
  n35 = Constant3Vector() {constant=<rgba 0 0 1 0>}
  n36 = TextureSample(UVs=@n12) {samplerSource=SSM_Wrap_WorldGroupSettings, samplerType=SAMPLERTYPE_Normal, texture=/Game/Global/Textures/Default/PointSplatter/Bubbles_N}
  n37 = TextureSample(UVs=@n15) {samplerSource=SSM_Wrap_WorldGroupSettings, samplerType=SAMPLERTYPE_Normal, texture=&Bubbles_N}
  n38 = MaterialFunctionCall(BaseNormal=@n36.RGB, AdditionalNormal=@n37.RGB) {materialFunction=&BlendAngleCorrectedNormals}
  n39 = ComponentMask(None=@n17) {g=true}
  n40 = Multiply(A=@n3, B=Saturate(None=Multiply(A=@n39) {constB=2}))
  NamedRerouteDeclaration(None=Multiply(
    A=LinearInterpolate(A=@n35, B=@n38.Result, Alpha=@n40)
    B=VectorParameter() {group=Potion, parameterName=Potion-BubbleNormalAmt, defaultValue=<rgba 1 1 1 1>}.RGB
   )) {name=Normal_OUT, nodeColor=<rgba 1 0.2812 0 1>}
  n41 = Subtract(A=Time(), B=ScalarParameter() {defaultValue=99999997952, group=CrawlyTech Backpack, parameterName=CT_GameTimeFromBP})
  n42 = ScalarParameter() {defaultValue=3.75, group=CrawlyTech Backpack, parameterName=CT_DilationDuration}
  n43 = CurveAtlasRowParameter(CurveTime=Saturate(None=Divide(A=@n41, B=@n42))) {
   atlas=/BRCosmetics/Accessories/FORT_Backpacks/Backpack_M_MED_CrawlyTech/Textures/CA_CrawlyTech_Backpack
   curve=/BRCosmetics/Accessories/FORT_Backpacks/Backpack_M_MED_CrawlyTech/Textures/C_CrawlyTech_Backpack
   group=CrawlyTech Backpack
   parameterName=CT_DilationAnim
  }
  NamedRerouteDeclaration(None=StaticSwitchParameter(True=ScalarParameter() {defaultValue=1, parameterName=DEBUG_EyeDilation}, False=@n43.R) {parameterName=DEBUG_Reactivity}) {name=DilationProgress, nodeColor=<rgba 0 0.1094 1 1>}
  n44 = PixelNormalWS()
  n45 = ScalarParameter() {defaultValue=2, group=CrawlyTech Backpack, parameterName=FresnelExponent, sortPriority=10}
  n46 = StaticBool()
  n47 = ScalarParameter() {group=CrawlyTech Backpack, parameterName=FresnelDarkLevel, sortPriority=11}
  n48 = ScalarParameter() {defaultValue=1, group=CrawlyTech Backpack, parameterName=FresnelBrightLevel, sortPriority=12}
  NamedRerouteDeclaration(None=Saturate(None=MaterialFunctionCall(Normal Vector=@n44, Power=@n45, Use Cheap contrast=@n46, Cheap contrast dark=@n47, Cheap contrast bright=@n48) {materialFunction=&Fresnel_Function}.Result)) {name=Fresnel_01, nodeColor=<rgba 0 1 0.4844 1>}
  NamedRerouteDeclaration(None=@n3) {name=Debug, nodeColor=<rgba 0.875 0 1 1>}
  NamedRerouteUsage()
 ]
}
