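# Parent material for the CrawlyTech backpack cosmetic, adding reactive dilating eyes and emissive effects between the 2023 character parent PreFX and PostFX functions.
# Time since a Blueprint-supplied game time samples a curve atlas for dilation progress, with a debug override.
# Eye UVs scale around left and right centers within sphere masks, feeding PreFX texture lookups.
# Sphere masks also build eye highlights and shadows, while emissive combines liquid color, normal-based fake light and highlights.
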Material /BRCosmetics/Accessories/FORT_Backpacks/Backpack_M_MED_CrawlyTech/Materials/M_CrawlyTech_Backpack_Parent
graph {
 properties {
  bUseMaterialAttributes=true
  bUsedWithMeshDeformer=true
  bUsedWithMorphTargets=true
  bUsedWithSkeletalMesh=true
  pixelDepthOffsetMode=PDOM_Legacy
  shadingModel=MSM_FromMaterialExpression
  usedShadingModels=Default Lit | Subsurface | Clear Coat | Subsurface Profile | Cloth
 }
 nodes [
  n1 = Time()
  Frac(None=@n1)
  n2 = ScalarParameter() {defaultValue=0.012, group=CrawlyTech Backpack, parameterName=Eye UV Highlight Size, sortPriority=126}
  LinearInterpolate(
   A=@n2
   B=ScalarParameter() {defaultValue=0.012, group=CrawlyTech Backpack, parameterName=Eye UV Highlight Size D, sortPriority=126}
   Alpha=NamedRerouteUsage()
  )
  n3 = TextureCoordinate()
  n4 = NamedRerouteUsage()
  n5 = ComponentMask(None=VectorParameter() {group=CrawlyTech Backpack, parameterName=EyePos_L, defaultValue=<rgba 0.872 0.4165 1024 1>}.RGB) {g=true, r=true}
  n6 = Subtract(A=Add(A=Divide(A=@n3, B=@n4), B=@n5), B=Divide(A=@n5, B=@n4) {desc=Scale the center offset by the scaling factor}) {desc=Re offset to maintain center}
  n7 = ScalarParameter() {defaultValue=0.2, parameterName=DilationRadius}
  n8 = TextureCoordinate()
  n9 = NamedRerouteUsage()
  n10 = ComponentMask(None=VectorParameter() {group=CrawlyTech Backpack, parameterName=EyePos_R, defaultValue=<rgba 0.904 0.634 1024 1>}.RGB) {g=true, r=true}
  n11 = Subtract(A=Add(A=Divide(A=@n8, B=@n9), B=@n10), B=Divide(A=@n10, B=@n9) {desc=Scale the center offset by the scaling factor}) {desc=Re offset to maintain center}
  n12 = TextureCoordinate()
  n13 = LinearInterpolate(
   A=LinearInterpolate(A=@n3, B=@n6, Alpha=SphereMask(A=@n3, B=@n5, Radius=@n7) {attenuationRadius=0.1, hardnessPercent=0})
   B=LinearInterpolate(A=@n8, B=@n11, Alpha=SphereMask(A=@n8, B=@n10, Radius=@n7) {attenuationRadius=0.1, hardnessPercent=0})
   Alpha=Step(X=ComponentMask(None=@n12) {g=true}) {constX=0.5, constY=0.515}
  )
  n14 = MaterialFunctionCall(Overide Texture Set UVs=@n13, Override Basic Emissive UVs=@n13) {materialFunction=&MF_CharacterParent_2023_PreFX}
  n15 = GetMaterialAttributes(None=@n14.Result) {attributeGetTypes [#1, #2, #3]}
  MaterialFunctionCall(
   Material=SetMaterialAttributes!(MaterialAttributes=@n15.MaterialAttributes, Base Color, Emissive Color=NamedRerouteUsage(), Roughness) {attributeSetTypes [#1, #3, #2]}
   M=@n14.M
   Pre Skinned Local Position=@n14.M
   Distance Blend=@n14.M
   Fresnel=NamedRerouteDeclaration(None=@n14.Fresnel) {name=Fresnel, nodeColor=<rgba 0 1 0.7656 1>}
  ) {materialFunction=&MF_CharacterParent_2023_PostFX}
  MaterialExpressionSubtract_5 = Subtract(A=CameraPositionWS(), B=TransformPosition(None=Constant3Vector() {constant=<rgba 0 -10 0 1>}) {transformType=TRANSFORMPOSSOURCE_World})
  n16 = Custom!(x=@MaterialExpressionSubtract_5) {
   code="length(x)"
   inputs [{
     inputName=x
     input {expression=&self:MaterialExpressionSubtract_5}
    }]
  }
  n17 = LinearInterpolate(
   B=ScalarParameter() {defaultValue=2, parameterName=EWR}
   Alpha=InverseLinearInterpolate(A=ScalarParameter() {parameterName=ogrjei}, B=ScalarParameter() {defaultValue=512, parameterName=InvLerpFar}, Value=@n16) {constB=2048}
  ) {constA=1, constB=2}
  n18 = Power(Base=Divide(A=PixelDepth().R, B=Multiply(A=@n16, B=@n17) {constB=2}), Exp=ScalarParameter() {defaultValue=2.5, parameterName=Power})
  Multiply(A=NamedRerouteUsage(), B=Saturate(None=SmoothStep(Value=@n18) {constMax=0.9, constMin=0.6}))
  n19 = Subtract(A=@n1, B=ScalarParameter() {defaultValue=99999997952, group=CrawlyTech Backpack, parameterName=CT_GameTimeFromBP})
  n20 = ScalarParameter() {defaultValue=3.75, group=CrawlyTech Backpack, parameterName=CT_DilationDuration}
  n21 = CurveAtlasRowParameter(CurveTime=Saturate(None=Divide(A=@n19, B=@n20))) {
   atlas=/BRCosmetics/Accessories/FORT_Backpacks/Backpack_M_MED_CrawlyTech/Textures/CA_CrawlyTech_Backpack
   curve=/BRCosmetics/Accessories/FORT_Backpacks/Backpack_M_MED_CrawlyTech/Textures/C_CrawlyTech_Backpack
   group=CrawlyTech Backpack
   parameterName=CT_DilationAnim
  }
  n22 = StaticSwitchParameter(True=ScalarParameter() {defaultValue=1, parameterName=DEBUG_EyeDilation}, False=@n21.R) {parameterName=DEBUG_Reactivity}
  NamedRerouteDeclaration(None=LinearInterpolate(
    A=ScalarParameter() {defaultValue=1, group=CrawlyTech Backpack, parameterName=EyeScaleStart}
    B=ScalarParameter() {defaultValue=2, group=CrawlyTech Backpack, parameterName=EyeScaleEnd}
    Alpha=NamedRerouteDeclaration(None=@n22) {name=DilationProgress, nodeColor=<rgba 0 0.1094 1 1>}
   )) {name=EyeScale, nodeColor=<rgba 1 0.8906 0 1>}
  NamedRerouteDeclaration(None=ComponentMask(None=@n14.Tech Art Mask) {b=true}) {name=LightMask, nodeColor=<rgba 0 0.6484 1 1>}
  NamedRerouteDeclaration(None=@n6) {name=L_EyeDilated_UV, nodeColor=<rgba 0 1 0.6016 1>}
  NamedRerouteDeclaration(None=@n11) {name=R_EyeDilated_UV, nodeColor=<rgba 1 0.0997 0.2825 1>}
  n23 = VectorParameter() {group=CrawlyTech Backpack, parameterName=Eye UV Highlight Pos L, sortPriority=120, defaultValue=<rgba 0.9 0.39 1024 1>}
  n24 = VectorParameter() {group=CrawlyTech Backpack, parameterName=Eye UV Highlight Pos L D, sortPriority=120, defaultValue=<rgba 0.885 0.4 1024 1>}
  n25 = NamedRerouteUsage()
  n26 = ComponentMask(None=LinearInterpolate(A=@n23.RGB, B=@n24.RGB, Alpha=@n25)) {g=true, r=true}
  n27 = LinearInterpolate(
   A=ScalarParameter() {defaultValue=0.0085, group=CrawlyTech Backpack, parameterName=ShadowWidth}
   B=ScalarParameter() {defaultValue=0.005, group=CrawlyTech Backpack, parameterName=ShadowWidth D}
   Alpha=NamedRerouteUsage()
  )
  n28 = Add(A=@n27, B=@n2)
  n29 = ScalarParameter() {defaultValue=0.9, group=CrawlyTech Backpack, parameterName=ShadowHardness}
  n30 = SphereMask(A=NamedRerouteUsage(), B=@n26, Radius=@n28, Hardness=@n29) {hardnessPercent=30}
  n31 = VectorParameter() {group=CrawlyTech Backpack, parameterName=Eye UV Highlight Pos R, sortPriority=120, defaultValue=<rgba 0.925 0.6 1024 1>}
  n32 = VectorParameter() {group=CrawlyTech Backpack, parameterName=Eye UV Highlight Pos R D, sortPriority=120, defaultValue=<rgba 0.92 0.618 1024 1>}
  n33 = ComponentMask(None=LinearInterpolate(A=@n31.RGB, B=@n32.RGB, Alpha=@n25)) {g=true, r=true}
  n34 = SphereMask(A=NamedRerouteUsage(), B=@n33, Radius=@n28, Hardness=@n29) {hardnessPercent=30}
  n35 = VectorParameter() {group=CrawlyTech Backpack, parameterName=Eye UV Highlight Pos L Small, sortPriority=120, defaultValue=<rgba 0.86 0.425 1024 1>}
  n36 = VectorParameter() {group=CrawlyTech Backpack, parameterName=Eye UV Highlight Pos L Small D, sortPriority=120, defaultValue=<rgba 0.865 0.42 1024 1>}
  n37 = NamedRerouteUsage()
  n38 = ComponentMask(None=LinearInterpolate(A=@n35.RGB, B=@n36.RGB, Alpha=@n37)) {g=true, r=true}
  n39 = LinearInterpolate(
   A=ScalarParameter() {defaultValue=0.008, group=CrawlyTech Backpack, parameterName=Eye UV Highlight Size 2, sortPriority=126}
   B=ScalarParameter() {defaultValue=0.007, group=CrawlyTech Backpack, parameterName=Eye UV Highlight Size 2 D, sortPriority=126}
   Alpha=NamedRerouteUsage()
  )
  n40 = Add(A=@n27, B=@n39)
  n41 = SphereMask(A=NamedRerouteUsage(), B=@n38, Radius=@n40, Hardness=@n29) {hardnessPercent=30}
  n42 = NamedRerouteUsage()
  n43 = VectorParameter() {group=CrawlyTech Backpack, parameterName=Eye UV Highlight Pos R Small, sortPriority=120, defaultValue=<rgba 0.895 0.64 1024 1>}
  n44 = VectorParameter() {group=CrawlyTech Backpack, parameterName=Eye UV Highlight Pos R Small D, sortPriority=120, defaultValue=<rgba 0.895 0.64 1024 1>}
  n45 = ComponentMask(None=LinearInterpolate(A=@n43.RGB, B=@n44.RGB, Alpha=@n37)) {g=true, r=true}
  n46 = SphereMask(A=@n42, B=@n45, Radius=@n40, Hardness=@n29) {hardnessPercent=30}
  NamedRerouteDeclaration(None=Add(A=Add(A=@n30, B=@n34), B=Add(A=@n41, B=@n46))) {name=EyesHighlightShadow, nodeColor=<rgba 1 0 0.8437 1>}
  NamedRerouteDeclaration(None=@n15.Base Color) {name=Basecolor_IN, nodeColor=<rgba 0 1 0.8359 1>}
  NamedRerouteDeclaration(None=Constant()) {name=DepthMask, nodeColor=<rgba 1 0 0.4453 1>}
  NamedRerouteDeclaration(None=ComponentMask(None=@n14.Tech Art Mask) {g=true}) {name=BugMask, nodeColor=<rgba 0 1 0.3906 1>}
  n47 = NamedRerouteUsage()
  n48 = ScalarParameter() {defaultValue=0.5, group=CrawlyTech Backpack, parameterName=Eye UV Highlight Hardness, sortPriority=126}
  n49 = Add(
   A=SphereMask(A=@n47, B=@n26, Radius=@n2, Hardness=@n48) {hardnessPercent=20}
   B=SphereMask(A=@n42, B=@n33, Radius=@n2, Hardness=@n48) {hardnessPercent=20}
  )
  n50 = ScalarParameter() {defaultValue=0.5, group=CrawlyTech Backpack, parameterName=Eye UV Highlight Hardness, sortPriority=126}
  n51 = Add(
   A=SphereMask(A=@n47, B=@n38, Radius=@n39, Hardness=@n50) {hardnessPercent=20}
   B=SphereMask(A=@n42, B=@n45, Radius=@n39, Hardness=@n50) {hardnessPercent=20}
  )
  NamedRerouteDeclaration(None=Saturate(None=Add(A=@n49, B=@n51))) {name=EyesHighlight, nodeColor=<rgba 0.2656 1 0 1>}
  NamedRerouteDeclaration(None=@n15.Emissive Color) {name=Emissive_IN, nodeColor=<rgba 0.5937 1 0 1>}
  n52 = Multiply(
   A=VectorParameter() {group=CrawlyTech Backpack, parameterName=LiquidColorE, defaultValue=<rgba 0.1757 0.1875 0.0521 1>}.RGB
   B=NamedRerouteUsage()
  )
  n53 = PreSkinnedNormal()
  n54 = Abs(None=DotProduct(A=VertexInterpolator(VS=@n53).PS, B=Constant3Vector() {constant=<rgba 0 0 1 0>}))
  n55 = VectorParameter() {group=CrawlyTech Backpack, parameterName=ReactiveEmissiveTint, sortPriority=61, defaultValue=<rgba 1.7946 2 0 1>}
  n56 = Multiply(A=Multiply(A=Power(Base=@n54), B=Constant() {r=0.35}), B=@n55.RGB)
  n57 = LinearInterpolate(
   A=NamedRerouteUsage()
   B=LinearInterpolate(A=@n52, B=Add(A=@n52, B=@n56) {desc=Fake light on reactivity}, Alpha=NamedRerouteUsage()) {constA=1, constB=1.5}
   Alpha=NamedRerouteUsage()
  )
  n58 = VectorParameter() {group=CrawlyTech Backpack, parameterName=EyeHighlightColor, defaultValue=<rgba 2 2 2 1>}
  n59 = NamedRerouteUsage()
  n60 = Multiply(A=NamedRerouteUsage(), B=NamedRerouteUsage())
  NamedRerouteDeclaration(None=Add(A=LinearInterpolate(A=@n57, B=@n58.RGB, Alpha=@n59), B=Multiply(A=@n55.RGB, B=@n60))) {name=Emissive_OUT, nodeColor=<rgba 1 0 0.9141 1>}
 ]
}
