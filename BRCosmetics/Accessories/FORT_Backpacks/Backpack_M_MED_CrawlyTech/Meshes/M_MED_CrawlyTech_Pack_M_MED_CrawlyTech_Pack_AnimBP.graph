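# Animation Blueprint for the CrawlyTech backpack, procedurally animating a bug with wings, feet and tail, plus physics jiggle.
# The EventGraph computes sine-based motion for bug, tail, wings and feet, with random delays toggling wing and foot bursts.
# Modify Bone nodes build the cached MovementPose.
# A linear-spring Anim Dynamics pass on dyn_bug is bypassed while skydiving, then an angular-spring pass with a pack-bone spherical limit always follows.
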AnimBlueprint /BRCosmetics/Accessories/FORT_Backpacks/Backpack_M_MED_CrawlyTech/Meshes/M_MED_CrawlyTech_Pack_M_MED_CrawlyTech_Pack_AnimBP
target_skeleton=&/M_MED_CrawlyTech_Pack_Skeleton_M_MED_CrawlyTech_Pack_Skeleton
graphs [
 AnimGraph {nodes [
   n1 = Anim Dynamics - Bone: dyn_bug!(ComponentPose=Local To Component(LocalPose=Use cached pose 'MovementPose'()), Alpha=1) -> (Pose) {
    angularDampingOverride=0.99
    angularSpringConstant=4000
    bLinearSpring=true
    bOverrideAngularDamping=true
    bUseGravityOverride=true
    bUsePlanarLimit=false
    linearSpringConstant=200
    alphaBoolBlend {blendOption=Linear}
    boundBone {boneName=dyn_bug}
    componentAppliedLinearAccClamp=<300 300 300>
    componentLinearAccScale=<0.1 0.1 0.1>
    componentLinearVelScale=<0.1 0.1 0.1>
    gravityOverride {z=-600}
    physicsBodyDefinitions [{
      collisionType=CustomSphere
      sphereCollisionRadius=6
      boundBone {boneName=dyn_bug}
      boxExtents=<40 40 40>
      constraintSetup {angularLimitsMax=<0 0 0>, angularLimitsMin=<0 0 0>, angularTarget=<1 0 0>, linearAxesMax=<1.5 0 0>, linearAxesMin=<-0.5 0 0>}
      localJointOffset=<-2 0 0>
     }]
   }
   n2 = Component To Local(ComponentPose=@n1)
   n3 = Use cached pose 'MovementPose'()
   n4 = Get bIsSkyDiving()
   n5 = Anim Dynamics - Bone: dyn_bug(
    ComponentPose=Local To Component(LocalPose=Blend Poses by bool(BlendPose_0=@n3, BlendPose_1=@n2, bActiveValue=@n4) {blendType=Linear})
    Alpha=1
   ) {
    angularDampingOverride=0.9
    angularSpringConstant=4000
    bAngularSpring=true
    bOverrideAngularDamping=true
    bUseGravityOverride=true
    bUsePlanarLimit=false
    bUseSphericalLimits=true
    alphaBoolBlend {blendOption=Linear}
    boundBone {boneName=dyn_bug}
    componentAppliedLinearAccClamp=<10000 10000 10000>
    componentLinearAccScale=<0.5 0.5 0.5>
    componentLinearVelScale=<0.5 0.5 0.5>
    gravityOverride {z=-800}
    physicsBodyDefinitions [{
      collisionType=CustomSphere
      sphereCollisionRadius=6
      boundBone {boneName=dyn_bug}
      boxExtents=<40 40 30>
      constraintSetup {angularLimitsMax=<20 40 30>, angularLimitsMin=<-20 0 0>, angularTarget=<1 0 0>}
      localJointOffset=<-2 0 -2>
     }]
    sphericalLimits [{limitRadius=10, limitType=Inner, drivingBone {boneName=pack}, sphereLocalOffset=<-2.0829 -7.5542 -4.6776>}]
   }
   Output Pose(Result=Component To Local(ComponentPose=@n5)) {layerGroup=None}
   n6 = "Transform (Modify) Bone - Bone: dyn_bug_mstr"!(ComponentPose=Local To Component(LocalPose=Input Pose()), Alpha=1, Translation="-1.6,-0.5,0") -> (Pose) {
    translationMode=BMM_Additive
    translationSpace=BCS_BoneSpace
    alphaBoolBlend {blendOption=Linear}
    boneToModify {boneName=dyn_bug_mstr}
    translation=<-1.6 -0.5>
   }
   n7 = "Transform (Modify) Bone - Bone: dyn_bug_mstr"(ComponentPose=@n6, Alpha=1, Translation="0,0,0") {translationMode=BMM_Additive, translationSpace=BCS_BoneSpace, alphaBoolBlend {blendOption=Linear}, boneToModify {boneName=dyn_bug_mstr}}
   n8 = "Transform (Modify) Bone - Bone: dyn_tail_00"!(ComponentPose=@n7, Alpha=1, Rotation="0,0,0") -> (Pose) {
    rotationMode=BMM_Additive
    rotationSpace=BCS_BoneSpace
    translationSpace=BCS_BoneSpace
    alphaBoolBlend {blendOption=Linear}
    boneToModify {boneName=dyn_tail_00}
   }
   n9 = "Transform (Modify) Bone - Bone: dyn_tail_01"!(ComponentPose=@n8, Alpha=1, Rotation="0,0,0") -> (Pose) {
    rotationMode=BMM_Additive
    rotationSpace=BCS_BoneSpace
    translationSpace=BCS_BoneSpace
    alphaBoolBlend {blendOption=Linear}
    boneToModify {boneName=dyn_tail_01}
   }
   n10 = "Transform (Modify) Bone - Bone: dyn_tail_02"!(ComponentPose=@n9, Alpha=1, Rotation="0,0,0") -> (Pose) {
    rotationMode=BMM_Additive
    rotationSpace=BCS_BoneSpace
    translationSpace=BCS_BoneSpace
    alphaBoolBlend {blendOption=Linear}
    boneToModify {boneName=dyn_tail_02}
   }
   n11 = "Transform (Modify) Bone - Bone: dyn_wing_l_00"!(ComponentPose=@n10, Alpha=1, Rotation="0,0,0") -> (Pose) {rotationMode=BMM_Additive, rotationSpace=BCS_BoneSpace, alphaBoolBlend {blendOption=Linear}, boneToModify {boneName=dyn_wing_l_00}}
   n12 = "Transform (Modify) Bone - Bone: dyn_wing_l_01"!(ComponentPose=@n11, Alpha=1, Rotation="0,0,0") -> (Pose) {rotationMode=BMM_Additive, rotationSpace=BCS_BoneSpace, alphaBoolBlend {blendOption=Linear}, boneToModify {boneName=dyn_wing_l_01}}
   n13 = "Transform (Modify) Bone - Bone: dyn_wing_r_00"!(ComponentPose=@n12, Alpha=1, Rotation="0,0,0") -> (Pose) {rotationMode=BMM_Additive, rotationSpace=BCS_BoneSpace, alphaBoolBlend {blendOption=Linear}, boneToModify {boneName=dyn_wing_r_00}}
   n14 = "Transform (Modify) Bone - Bone: dyn_wing_r_01"!(ComponentPose=@n13, Alpha=1, Rotation="0,0,0") -> (Pose) {rotationMode=BMM_Additive, rotationSpace=BCS_BoneSpace, alphaBoolBlend {blendOption=Linear}, boneToModify {boneName=dyn_wing_r_01}}
   n15 = "Transform (Modify) Bone - Bone: dyn_foot_01_l_mstr"!(ComponentPose=@n14, Alpha=1, Rotation="0,0,0") -> (Pose) {rotationMode=BMM_Additive, rotationSpace=BCS_BoneSpace, alphaBoolBlend {blendOption=Linear}, boneToModify {boneName=dyn_foot_01_l_mstr}}
   n16 = "Transform (Modify) Bone - Bone: dyn_foot_01_r_mstr"!(ComponentPose=@n15, Alpha=1, Rotation="0,0,0") -> (Pose) {rotationMode=BMM_Additive, rotationSpace=BCS_BoneSpace, alphaBoolBlend {blendOption=Linear}, boneToModify {boneName=dyn_foot_01_r_mstr}}
   n17 = "Transform (Modify) Bone - Bone: dyn_foot_02_l_mstr"!(ComponentPose=@n16, Alpha=1, Rotation="0,0,0") -> (Pose) {rotationMode=BMM_Additive, rotationSpace=BCS_BoneSpace, alphaBoolBlend {blendOption=Linear}, boneToModify {boneName=dyn_foot_02_l_mstr}}
   n18 = "Transform (Modify) Bone - Bone: dyn_foot_02_r_mstr"!(ComponentPose=@n17, Alpha=1, Rotation="0,0,0") -> (Pose) {rotationMode=BMM_Additive, rotationSpace=BCS_BoneSpace, alphaBoolBlend {blendOption=Linear}, boneToModify {boneName=dyn_foot_02_r_mstr}}
   n19 = "Transform (Modify) Bone - Bone: dyn_foot_03_l_mstr"!(ComponentPose=@n18, Alpha=1, Rotation="0,0,0") -> (Pose) {rotationMode=BMM_Additive, rotationSpace=BCS_BoneSpace, alphaBoolBlend {blendOption=Linear}, boneToModify {boneName=dyn_foot_03_l_mstr}}
   n20 = "Transform (Modify) Bone - Bone: dyn_foot_03_r_mstr"!(ComponentPose=@n19, Alpha=1, Rotation="0,0,0") -> (Pose) {rotationMode=BMM_Additive, rotationSpace=BCS_BoneSpace, alphaBoolBlend {blendOption=Linear}, boneToModify {boneName=dyn_foot_03_r_mstr}}
   Save cached pose 'MovementPose'(Pose=Component To Local(ComponentPose=@n20)) {cacheName=MovementPose}
  ]}
 EventGraph {nodes [
   n1 = Event BlueprintUpdateAnimation()
   > Set Delta_Time_X(Delta_Time_X:float=@n1.DeltaTimeX) -> (Output_Get:float)
   > n2 = Sequence!(execute) -> (then_0, then_1, then_2:Exec, then_3:Exec, then_4:Exec, then_5:Exec)
   n3 = "Sin (Radians)"(A=float * float(A=GetGameTimeInSeconds(), B=Get Foot_01_Speed_Multiplier() -> (Foot_01_Speed_Multiplier:double)) {is=title K2Node_PromotableOperator})
   n4 = Set Foot_01_Sin_float(
    execute=@n2.then_3
    Foot_01_Sin_float:double=FInterpTo(
     Current=Get Foot_01_Sin_float() -> (Foot_01_Sin_float:double)
     Target="Clamp (Float)"(Value=@n3, Min=0, Max=1)
     DeltaTime=Get Delta_Time_X() -> (Delta_Time_X:float)
     InterpSpeed=35
    )
   ) -> (Output_Get:double)
   n5 = "Sin (Radians)"(A=float * float(A=GetGameTimeInSeconds(), B=Get Foot_02_Speed_Multiplier() -> (Foot_02_Speed_Multiplier:double)))
   > n6 = Set Foot_02_Sin_float(Foot_02_Sin_float:double=FInterpTo(
     Current=Get Foot_02_Sin_float() -> (Foot_02_Sin_float:double)
     Target="Clamp (Float)"(Value=@n5, Min=0, Max=1)
     DeltaTime=Get Delta_Time_X()
     InterpSpeed=35
    )) -> (Output_Get:double)
   n7 = "Sin (Radians)"(A=float * float(A=GetGameTimeInSeconds(), B=Get Foot_03_Speed_Multiplier() -> (Foot_03_Speed_Multiplier:double)))
   > n8 = Set Foot_03_Sin_float(Foot_03_Sin_float:double=FInterpTo(
     Current=Get Foot_03_Sin_float() -> (Foot_03_Sin_float:double)
     Target="Clamp (Float)"(Value=@n7, Min=0, Max=1)
     DeltaTime=Get Delta_Time_X()
     InterpSpeed=35
    )) -> (Output_Get:double)
   > Set Foot_01_Rot(Foot_01_Rot:Rotator=MakeRotator(Yaw=float * float(A=@n4, B=-40))) -> (Output_Get:Rotator)
   > Set Foot_02_Rot(Foot_02_Rot:Rotator=MakeRotator(Yaw=float * float(A=@n6, B=-40))) -> (Output_Get:Rotator)
   > Set Foot_03_Rot(Foot_03_Rot:Rotator=MakeRotator(Yaw=float * float(A=@n8, B=-40))) -> (Output_Get:Rotator)
   n9 = Branch(execute=@n2.then_2, Condition=Get Foot_01_Animation_Enabled() -> (Foot_01_Animation_Enabled:bool))
   > Set Foot_01_Speed_Multiplier(Foot_01_Speed_Multiplier:double=20) -> (Output_Get:double)
   > Set Foot_02_Speed_Multiplier(Foot_02_Speed_Multiplier:double=15) -> (Output_Get:double)
   > Set Foot_03_Speed_Multiplier(Foot_03_Speed_Multiplier:double=10) -> (Output_Get:double)
   > Delay(Duration=RandomFloatInRange(Min=0.5, Max=1.5))
   > Set Foot_01_Animation_Enabled(Foot_01_Animation_Enabled:bool=false) -> (Output_Get:bool)
   Set Foot_01_Speed_Multiplier(execute=@n9.else, Foot_01_Speed_Multiplier=0)
   > Set Foot_02_Speed_Multiplier(Foot_02_Speed_Multiplier=0)
   > Set Foot_03_Speed_Multiplier(Foot_03_Speed_Multiplier=0)
   > Delay(Duration=RandomFloatInRange(Min=3, Max=5))
   > Set Foot_01_Animation_Enabled(Foot_01_Animation_Enabled=true)
   n10 = GetRealTimeSeconds()
   n11 = "Sin (Radians)"(A=float * float(A=float * float(A=1, B=-1.5), B=@n10))
   Set Translate_Bug(execute=@n2.then_4, Translate_Bug:Vector=MakeVector(X=float / float(A=@n11, B=1), Y=0)) -> (Output_Get:Vector)
   n12 = GetRealTimeSeconds()
   n13 = float * float(A=1, B=1)
   n14 = float * float(A="Sin (Radians)"(A=float * float(A=@n13, B=@n12)), B=10)
   n15 = Branch(execute=@n2.then_5, Condition="float < float"(A=@n14))
   > Set Rotate_Tail(Rotate_Tail:Rotator=MakeRotator(Yaw=@n14)) -> (Output_Get:Rotator)
   Set Rotate_Tail(execute=@n15.else, Rotate_Tail="0, 0, 0")
   n16 = "Sin (Radians)"(A=float * float(A=GetGameTimeInSeconds(), B=Get Wing_Speed_Multiplier() -> (Wing_Speed_Multiplier:double)))
   n17 = Set Wing_Sin_float(
    execute=@n2.then_1
    Wing_Sin_float:double=FInterpTo(
     Current=Get Wing_Sin_float() -> (Wing_Sin_float:double)
     Target="Clamp (Float)"(Value=@n16, Min=0, Max=1)
     DeltaTime=Get Delta_Time_X()
     InterpSpeed=35
    )
   ) -> (Output_Get:double)
   > Set Wing_L_Rot(Wing_L_Rot:Rotator=MakeRotator(Pitch=float * float(A=@n17, B=30))) -> (Output_Get:Rotator)
   > Set Wing_R_Rot(Wing_R_Rot:Rotator=MakeRotator(Pitch=float * float(A=@n17, B=-30))) -> (Output_Get:Rotator)
   n18 = Branch(execute=@n2.then_0, Condition=Get Wing_Animation_Enabled() -> (Wing_Animation_Enabled:bool))
   > Set Wing_Speed_Multiplier(Wing_Speed_Multiplier:double=42) -> (Output_Get:double)
   > Delay(Duration=RandomFloatInRange(Min=0.5, Max=1.5))
   > Set Wing_Animation_Enabled(Wing_Animation_Enabled:bool=false) -> (Output_Get:bool)
   Set Wing_Speed_Multiplier(execute=@n18.else, Wing_Speed_Multiplier=0)
   > Delay(Duration=RandomFloatInRange(Min=5, Max=7))
   > Set Wing_Animation_Enabled(Wing_Animation_Enabled=true)
   n19 = float * float(A=1, B=-2)
   float / float(A="Sin (Radians)"(A=float * float(A=@n19, B=@n10)), B=6)
   n20 = float * float(A=1, B=-2)
   float / float(A="Sin (Radians)"(A=float * float(A=@n20, B=@n10)), B=6)
  ]}
]
variables [
 Delta_Time_X {type=number}
 Foot_01_Animation_Enabled {type=boolean}
 Foot_01_Rot {type=Rotator}
 Foot_01_Sin_float {type=number}
 Foot_01_Speed_Multiplier {type=number}
 Foot_02_Animation_Enabled {type=boolean}
 Foot_02_Rot {type=Rotator}
 Foot_02_Sin_float {type=number}
 Foot_02_Speed_Multiplier {type=number}
 Foot_03_Animation_Enabled {type=boolean}
 Foot_03_Rot {type=Rotator}
 Foot_03_Sin_float {type=number}
 Foot_03_Speed_Multiplier {type=number}
 Rotate_Tail {type=Rotator}
 Translate_Bug {type=Vector}
 Wing_Animation_Enabled {type=boolean}
 Wing_L_Rot {type=Rotator}
 Wing_R_Rot {type=Rotator}
 Wing_Sin_float {type=number}
 Wing_Speed_Multiplier {type=number}
]
